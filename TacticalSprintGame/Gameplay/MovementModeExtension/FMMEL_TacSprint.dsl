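; Movement mode extension logic for tactical sprint that drives the local player's camera effects while the extension is active.
; On activation it caches the player controller and starts FOV flare and looping noise camera shakes on its camera manager.
; First-person camera skips the intro shake and sets the speed-lines variable to its default.
; Deactivation stops the stored shakes non-immediately and clears the cached controller.
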(event EventOnExtensionActivate
  (params (RuntimeContext :type MMERuntime%20Context%20Structure%20%28by%20ref%29))
  (if (Variables|FortMovementMode_BaseExtLogic|GetbIsLocalPlayer)
    (bind _asfort_player_pawn (Utilities|Casting|CastToFortPlayerPawn (Variables|FortMovementMode_BaseExtLogic|GetOwnerPawn)))
    (Utilities|IsValid _asfort_player_pawn
      (:Is%20Valid
        (Utilities|IsValid (Fortnite|GetFortPlayerControllerfromActor _asfort_player_pawn)
          (:Is%20Valid
            (bind _output_get (Variables|Default|SetFortPlayerController (Fortnite|GetFortPlayerControllerfromActor _asfort_player_pawn)))
            (bind _output_get_2 (Variables|Default|SetIsUsingFPCamera (Camera|IsUsingFirstPersonCamera _output_get)))
            (if _output_get_2
              (Variables|FortMovementMode_ELTacSprint|SetGCSpeedLines))
            (Utilities|IsValid (Variables|/Script/Engine.PlayerController|GetPlayerCameraManager _output_get)
              (:Is%20Valid
                (bind _returnvalue (CameraShakes|StartCameraShake :self (Variables|/Script/Engine.PlayerController|GetPlayerCameraManager _output_get) :ShakeClass "/TacticalSprintGame/Gameplay/Camera/CameraShake_tacticalSprint_FOVFlare.CameraShake_tacticalSprint_FOVFlare_C" :Scale 1.0 :PlaySpace "CameraLocal"))
                (Variables|Default|SetFOVFlare _returnvalue)
                (bind _returnvalue_2 (CameraShakes|StartCameraShake :self (Variables|/Script/Engine.PlayerController|GetPlayerCameraManager _output_get) :ShakeClass "/TacticalSprintGame/Gameplay/Camera/CameraShake_TacticalSprint_Loop.CameraShake_TacticalSprint_Loop_C" :Scale 0.8 :PlaySpace "CameraLocal"))
                (Variables|Default|SetShakeNoise _returnvalue_2)
                ;[1 1 1] Skip playing this anim if using First Person Camera
                (if (not (Variables|Default|GetIsUsingFPCamera))
                  (bind _returnvalue_3 (CameraShakes|StartCameraShake :self (Variables|/Script/Engine.PlayerController|GetPlayerCameraManager _output_get) :ShakeClass "/TacticalSprintGame/Gameplay/Camera/CameraShake_tacticalSprint_StartAnim.CameraShake_tacticalSprint_StartAnim_C" :Scale 1.0 :PlaySpace "CameraLocal"))
                  (Variables|Default|SetTacSprintIntroShake _returnvalue_3))))))))))

(event EventOnExtensionDeactivate
  (params (RuntimeContext :type MMERuntime%20Context%20Structure%20%28by%20ref%29))
  (bind _fortplayercontroller (Variables|Default|GetFortPlayerController))
  (bind _playercameramanager (Variables|/Script/Engine.PlayerController|GetPlayerCameraManager _fortplayercontroller))
  (Utilities|IsValid _fortplayercontroller
    (:Is%20Valid
      (Utilities|IsValid _playercameramanager
        (:Is%20Valid
          (CameraShakes|StopCameraShake _playercameramanager (Variables|Default|GetFOVFlare) false)
          (CameraShakes|StopCameraShake _playercameramanager (Variables|Default|GetShakeNoise) false)
          (if (not (Variables|Default|GetIsUsingFPCamera))
            (CameraShakes|StopCameraShake _playercameramanager (Variables|Default|GetTacSprintIntroShake) false))
          (Variables|Default|SetFortPlayerController 0))))))
